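; BP_PistolCanary_SlideComponent is a component on a Fort player pawn that forces the player to slide along a set direction.
; On BeginPlay it caches the owning pawn and the first BP_PistolCanary_Slide_Direction actor it finds.
; Each tick it pushes movement input along that actor's forward vector so players can't move against it, enables sliding, and after SlideDelay triggers crouch whenever not sliding.
; A Destroy event removes the component.
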(event EventBeginPlay
  (bind _asfort_player_pawn (Utilities|Casting|CastToFortPlayerPawn (Components|GetOwner)))
  (Variables|Default|SetPlayerPawn _asfort_player_pawn)
  (bind _returnvalue (Actor|GetActorOfClass "/PistolCanary/Gameplay/Abilities/BP_PistolCanary_Slide_Direction.BP_PistolCanary_Slide_Direction_C"))
  (Utilities|IsValid _returnvalue
    (:Is%20Valid
      (Variables|Default|SetCurrentSlideDirection _returnvalue))))

(event EventTick
  (params (DeltaSeconds :type Float%20%28single-precision%29))
  (Variables|Default|SetTimeElapsed (+ (Variables|Default|GetTimeElapsed) DeltaSeconds))
  (Utilities|IsValid (Variables|Default|GetPlayerPawn)
    (:Is%20Valid
      (Utilities|IsValid (Variables|Default|GetCurrentSlideDirection)
        (:Is%20Valid
          (CallFunction|AddMovementTowardsSlideDestination)
          (if (> (Variables|Default|GetTimeElapsed) (Variables|Default|GetSlideDelay))
            (CallFunction|TrySetSlide)))))))

(event Custom|Destroy
  (Components|DestroyComponent))

(fn AddMovementTowardsSlideDestination
  ;[0.15 0.15 0.15 0.5 15] Add input to the player towards the slide direction automatically so they can't move against it
  (Pawn|Input|AddMovementInput (Variables|Default|GetPlayerPawn) (Transformation|GetActorForwardVector (Variables|Default|GetCurrentSlideDirection)) (Variables|Default|GetMovementMultiplier) true)
  (FortPawn|SetCanSlide (Variables|Default|GetPlayerPawn) true))

(fn TrySetSlide
  (bind _playerpawn (Variables|Default|GetPlayerPawn))
  (if (not (Variables|/Script/FortniteGame.FortPawn|GetbIsSliding (Variables|Default|GetPlayerPawn)))
    (Class|FortPlayerPawn|ToggleCrouch _playerpawn true)
    (Class|FortPlayerPawn|SetCrouch _playerpawn true true)))
