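; GA_PistolCanary_TentacleEnergyBlast is a gameplay ability, named for a tentacle energy blast attack of PistolCanary.
; It is currently a stub: activation only prints a debug string tagged "BGA_TB:: Energy Blast".
; The end-ability event is empty, and the graph contains no call to commit or end the ability.
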(event Ability|EventActivateAbility
  (Development|PrintString "BGA_TB:: Energy Blast "))

(event Ability|EventOnEndAbility
  (params (bWasCancelled :type Boolean)))
